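# UI material for a Quotient item shop tile that layers a positionable character texture over a themed, specular-patterned background.
# The Color parameter's RGBA channels drive chained lerps that pick between teal, blue, orange and red three-color gradient backgrounds.
# A border mask combines pixel-size-scaled edge steps with top/bottom gradients and blurred noise, offset in quantized time steps for choppy animated edges, and multiplies into the background.
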Material /OfferCatalog/Art/Quotient/M_UI_ShopTile_Quotient
graph {
 properties {bUsedWithStaticMesh=false, materialDomain=MD_UI, pixelDepthOffsetMode=PDOM_Legacy, usedShadingModels=Default Lit}
 nodes [
  n1 = NamedRerouteUsage()
  n2 = AppendVector(A=ScalarParameter() {parameterName=Character X Position}, B=ScalarParameter() {parameterName=Character Y Position})
  n3 = TextureSampleParameter2D(UVs=MaterialFunctionCall(UVs=Add(A=@n1, B=@n2), Texture Scale=Constant() {r=1}) {materialFunction=&ScaleUVsByCenter}.UVs) {parameterName=Character, texture=/OfferCatalog/Textures/BattleRoyale/FeaturedItems/Outfit/S24/T-AthenaSoldiers-Character-FearlessFlightHero}
  LinearInterpolate(A=NamedRerouteUsage(), B=@n3.RGB, Alpha=@n3.RGB)
  n4 = MaterialFunctionCall(Float2=VectorParameter() {parameterName=Native Pixel Size, defaultValue=<rgba 300 13 0 1>}.RGB) {materialFunction=&BreakOutFloat2Components}
  n5 = Constant() {r=0.05}
  NamedRerouteDeclaration(None=Divide(A=Multiply(A=@n5, B=@n4.R), B=NamedRerouteUsage())) {name=Thickness X, nodeColor=<rgba 1 0 0.5102 1>}
  NamedRerouteDeclaration(None=MaterialFunctionCall() {materialFunction=&GetUserInterfaceUV}.9-Slice UV) {name=UVScaling, nodeColor=<rgba 1 0 0.9844 1>}
  NamedRerouteDeclaration(None=MaterialFunctionCall(
    A=Constant3Vector() {constant=<rgba 0 0.0698 0.0938 1>}
    B=Constant3Vector() {constant=<rgba 0 0.2188 0.1949 1>}
    C=Constant3Vector() {constant=<rgba 0.0814 0.625 0.5727 1>}
    Alpha=NamedRerouteUsage()
   ) {materialFunction=&3ColorBlend}.Result) {name=TealTileBG, nodeColor=<rgba 0 1 0.2969 1>}
  NamedRerouteDeclaration(None=MaterialFunctionCall(
    A=Constant3Vector() {constant=<rgba 0 0.0889 0.5208 1>}
    B=Constant3Vector() {constant=<rgba 0.0112 0.3937 0.8073 1>}
    C=Constant3Vector() {constant=<rgba 0.1968 0.6661 0.9688 1>}
    Alpha=NamedRerouteUsage()
   ) {materialFunction=&3ColorBlend}.Result) {name=BlueTileBG, nodeColor=<rgba 0 0.2467 1 1>}
  NamedRerouteDeclaration(None=MaterialFunctionCall(
    A=Constant3Vector() {constant=<rgba 1 0.2268 0.0781 1>}
    B=Constant3Vector() {constant=<rgba 1 0.1007 0.0312 1>}
    C=Constant3Vector() {constant=<rgba 1 0.4785 0.1458 1>}
    Alpha=NamedRerouteUsage()
   ) {materialFunction=&3ColorBlend}.Result) {name=OrangeTileBG, nodeColor=<rgba 1 0.3243 0 1>}
  NamedRerouteDeclaration(None=MaterialFunctionCall(
    A=Constant3Vector() {constant=<rgba 1 0.1302 0.1549 1>}
    B=Constant3Vector() {constant=<rgba 0.8125 0.0296 0.0604 1>}
    C=Constant3Vector() {constant=<rgba 0.9948 0.4093 0.4223 1>}
    Alpha=NamedRerouteUsage()
   ) {materialFunction=&3ColorBlend}.Result) {name=RedTileBG, nodeColor=<rgba 1 0.0282 0 1>}
  n6 = Constant()
  n7 = Frac(None=Multiply(A=Time(), B=Constant() {r=0.25}))
  n8 = Constant() {r=12.5}
  n9 = Multiply(A=Ceil(None=Multiply(A=@n7, B=@n8)), B=Multiply(A=@n8) {constB=0.01})
  n10 = MaterialFunctionCall(UVs=Multiply(A=NamedRerouteUsage()) {constB=15}, "Rotation Angle (0-1)"=Constant() {r=0.75}) {materialFunction=&CustomRotator}
  n11 = AppendVector(A=Constant() {r=5}, B=Constant() {r=0.01})
  NamedRerouteDeclaration(None=Add(A=AppendVector(A=@n6, B=@n9), B=Multiply(A=@n10.Rotated Values, B=@n11))) {name=XBorder Time, nodeColor=<rgba 1 0 0.8906 1>}
  n12 = MaterialFunctionCall() {materialFunction=&MF_UI_UVScaling}
  n13 = TextureSample(UVs=@n12.Scaled To Fill) {samplerType=SAMPLERTYPE_Alpha, texture=&/T_UI_BelongTreaty_BkgSpecular_512}
  NamedRerouteDeclaration(None=@n13.R) {name=ShopTileBG Alpha, nodeColor=<rgba 0 1 0.3438 1>}
  n14 = MaterialFunctionCall!(
   UVs=MaterialFunctionCall(UVs=NamedRerouteUsage(), "Rotation Angle (0-1)"=Constant() {r=0.75}) {materialFunction=&CustomRotator}.Rotated Values
   Subtract=Constant() {r=0.92}
   Falloff=Constant() {r=50}
   Intensity=Constant() {r=3}
  ) {materialFunction=/Game/UI/Frontend/ItemShop/Art/_MaterialFunctions_ShopTheming/MF_UI_TopBottomGradient}
  n15 = Time()
  n16 = Constant() {r=0.3}
  n17 = Constant() {r=9}
  n18 = Multiply(A=Frac(None=Multiply(A=@n15, B=@n16)), B=@n17)
  n19 = AppendVector(A=Constant(), B=Multiply(A=Ceil(None=@n18), B=Multiply(A=@n17) {constB=0.01}))
  n20 = Multiply(A=NamedRerouteUsage(), B=AppendVector(A=Constant() {r=25}, B=Constant() {r=0.05}))
  n21 = SmoothStep(Value=TextureSample(UVs=Add(A=@n19, B=@n20)) {samplerType=SAMPLERTYPE_Grayscale, texture=/FortWater/Textures/LowResBlurredNoise}.R) {constMin=0.25}
  NamedRerouteDeclaration(None=Max(A=OneMinus(None=@n14.Result), B=SmoothStep(Value=@n21) {constMax=0.08, constMin=0.11})) {name=TextureBorderX, nodeColor=<rgba 1 0.6094 0 1>}
  n22 = MaterialFunctionCall() {materialFunction=&GetUserInterfaceUV}
  NamedRerouteDeclaration(None=ComponentMask(None=@n22.Pixel Size) {r=true}) {name=PixelSizeX, nodeColor=<rgba 1 0.0389 0.1086 1>}
  NamedRerouteDeclaration(None=ComponentMask(None=@n22.Pixel Size) {g=true}) {name=PixelSize Y, nodeColor=<rgba 1 0.3984 0 1>}
  NamedRerouteDeclaration(None=Divide(A=Multiply(A=@n4.G, B=@n5), B=NamedRerouteUsage())) {name=Thickness Y, nodeColor=<rgba 1 0.8901 0 1>}
  n23 = NamedRerouteUsage()
  n24 = MaterialFunctionCall(UVs=@n23, Subtract=Constant() {r=0.93}, Falloff=Constant() {r=50}, Intensity=Constant() {r=3}) {materialFunction=&MF_UI_TopBottomGradient}
  n25 = SmoothStep(Value=TextureSample(UVs=NamedRerouteUsage()) {samplerType=SAMPLERTYPE_Grayscale, texture=&LowResBlurredNoise}.R) {constMin=0.25}
  NamedRerouteDeclaration(None=Max(A=OneMinus(None=@n24.Result), B=SmoothStep(Value=@n25) {constMax=0.08, constMin=0.11})) {name=TextureTopBottom, nodeColor=<rgba 0 0.0156 1 1>}
  NamedRerouteDeclaration(None=@n12.Scaled To Fill) {name=CharacterRenderMask, nodeColor=<rgba 0 1 0.9063 1>}
  n26 = VectorParameter() {group=BG Color, parameterName=Color, defaultValue=<rgba 1 1 1 1>}
  n27 = LinearInterpolate(B=NamedRerouteUsage(), Alpha=@n26.R)
  n28 = NamedRerouteUsage()
  n29 = LinearInterpolate(
   A=LinearInterpolate(A=LinearInterpolate(A=@n27, B=@n28, Alpha=@n26.G), B=NamedRerouteUsage(), Alpha=@n26.B)
   B=NamedRerouteUsage()
   Alpha=@n26.A
  )
  n30 = Constant()
  n31 = NamedRerouteUsage()
  n32 = ComponentMask(None=NamedRerouteUsage()) {r=true}
  n33 = Step(Y=@n31, X=@n32)
  n34 = Step(Y=@n31, X=OneMinus(None=@n32))
  n35 = Multiply(A=NamedRerouteUsage()) {constB=20}
  n36 = ComponentMask(None=@n23) {g=true}
  n37 = Step(Y=@n35, X=@n36)
  n38 = Step(Y=@n35, X=OneMinus(None=@n36))
  n39 = Min(A=Min(A=NamedRerouteUsage(), B=Min(A=@n33, B=@n34)), B=Min(A=Min(A=@n37, B=@n38), B=NamedRerouteUsage()))
  NamedRerouteDeclaration(None=Multiply(A=LinearInterpolate(A=@n30, Alpha=@n39), B=Multiply(A=@n29, B=@n13.R))) {name=AddingOutlinetoBG, nodeColor=<rgba 0.2422 1 0 1>}
  ScalarParameter() {defaultValue=0.005, parameterName=Outline Thickness}
 ]
}
